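# Physics asset for the EmeraldGlassGreen backpack cosmetic, defining collision and joint limits for its dangling limb bones.
# A two-capsule pack body anchors three chains of small capsules: a two-segment right arm and two leg-knee-foot legs.
# Elbow and knees bend on a single swing axis, while hips and shoulder favor one axis.
# Twist is locked on every joint except the feet, which allow slight twist.
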PhysicsAsset /BRCosmetics/Accessories/FORT_Backpacks/Backpack_M_MED_EmeraldGlassGreen/Meshes/M_MED_EmeraldGlassGreen_Pack_Physics
bodies {
 dyn_arm_R [{shapeType=Capsule, length=2, radius=2, shapeName=dyn_arm_R_capsule, center=<2 0 0>, rotation=<rot 90 0 0>}]
 dyn_foot_L [{shapeType=Capsule, length=2, radius=2, shapeName=dyn_foot_L_capsule, center=<3 0 0>, rotation=<rot 90 0 0>}]
 dyn_foot_R [{shapeType=Capsule, length=2, radius=2, shapeName=dyn_foot_R_capsule, center=<3 0 0>, rotation=<rot 90 0 0>}]
 dyn_forarm_R [{shapeType=Capsule, length=2, radius=2, shapeName=dyn_forarm_R_capsule, center=<2 0 0>, rotation=<rot 90 0 0>}]
 dyn_knee_L [{shapeType=Capsule, length=2, radius=2, shapeName=dyn_knee_L_capsule, center=<2 0 0>, rotation=<rot 90 0 0>}]
 dyn_knee_R [{shapeType=Capsule, length=2, radius=2, shapeName=dyn_knee_R_capsule, center=<2 0 0>, rotation=<rot 90 0 0>}]
 dyn_leg_L [{shapeType=Capsule, length=2, radius=2, shapeName=dyn_leg_L_capsule, center=<2 0 0>, rotation=<rot 90 0 0>}]
 dyn_leg_R [{shapeType=Capsule, length=2, radius=2, shapeName=dyn_leg_R_capsule, center=<2 0 0>, rotation=<rot 90 0 0>}]
 pack [
  {
   shapeType=Capsule
   length=4.5457
   radius=14.1479
   shapeName=pack_capsule
   center=<6.6625 18.3351 4.3946>
   rotation=<rot -0.4189 124.0271 -164.1931>
  }
  {
   shapeType=Capsule
   length=20.9375
   radius=7.75
   shapeName=pack_capsule
   center=<-5.0268 4.6665 -1.7168>
   rotation=<rot 15.9857 -2.8471 -4.4392>
  }
 ]
}
constraints [
 {bone1Name=dyn_forarm_R, bone2Name=dyn_arm_R, swing1LimitDegrees=0, swing2LimitDegrees=37.5, twistLimitDegrees=0}
 {bone1Name=dyn_knee_L, bone2Name=dyn_leg_L, swing1LimitDegrees=25, swing2LimitDegrees=0, twistLimitDegrees=0}
 {bone1Name=dyn_knee_R, bone2Name=dyn_leg_R, swing1LimitDegrees=25, swing2LimitDegrees=0, twistLimitDegrees=0}
 {bone1Name=dyn_leg_L, bone2Name=pack, swing1LimitDegrees=60, swing2LimitDegrees=8, twistLimitDegrees=0}
 {bone1Name=dyn_leg_R, bone2Name=pack, swing1LimitDegrees=60, swing2LimitDegrees=8, twistLimitDegrees=0}
 {bone1Name=dyn_arm_R, bone2Name=pack, swing1LimitDegrees=45, swing2LimitDegrees=10, twistLimitDegrees=0}
 {bone1Name=dyn_foot_R, bone2Name=dyn_knee_R, swing1LimitDegrees=30, swing2LimitDegrees=30, twistLimitDegrees=2}
 {bone1Name=dyn_foot_L, bone2Name=dyn_knee_L, swing1LimitDegrees=30, swing2LimitDegrees=30, twistLimitDegrees=2}
]
