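# Physics asset for the left arm of the EmeraldGlassGreen backpack cosmetic, defining bodies and joints for its dangling arm bones.
# Small capsules on the upper arm and forearm chain to the pack body, which uses two larger capsules.
# The shoulder joint allows wide swing on one axis and narrow swing on the other, while the elbow locks its second swing axis like a hinge; both set zero twist limits.
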PhysicsAsset /BRCosmetics/Accessories/FORT_Backpacks/Backpack_M_MED_EmeraldGlassGreen/Meshes/M_MED_EmeraldGlassGreen_Pack_Physics_LeftARM
bodies {
 dyn_arm_L [{shapeType=Capsule, length=2, radius=2, shapeName=dyn_arm_L_capsule, center=<2 0 0>, rotation=<rot 90 -157.1874 -157.3857>}]
 dyn_forarm_L [{shapeType=Capsule, length=2, radius=2, shapeName=dyn_forarm_L_capsule, center=<2 0 0>, rotation=<rot 90 0 0>}]
 pack [
  {
   shapeType=Capsule
   length=13.0457
   radius=5.6479
   shapeName=pack_capsule
   center=<9.2449 8.1709 3.4624>
   rotation=<rot 4.1302 107.7755 -74.7234>
  }
  {
   shapeType=Capsule
   length=20.9375
   radius=7.75
   shapeName=pack_capsule
   center=<-5.0268 4.6665 -1.7168>
   rotation=<rot 15.9857 -2.8471 -4.4392>
  }
 ]
}
constraints [
 {bone1Name=dyn_forarm_L, bone2Name=dyn_arm_L, swing1LimitDegrees=37.5, swing2LimitDegrees=0, swing2Motion=Locked, twistLimitDegrees=0}
 {bone1Name=dyn_arm_L, bone2Name=pack, swing1LimitDegrees=45, swing2LimitDegrees=10, twistLimitDegrees=0}
]
