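# Animation Blueprint for the EmeraldGlassGreen backpack cosmetic, adding physics sway and a reaction to the player's kills.
# Two RigidBody nodes simulate the pack and left arm with gravity override, their weights varying with skydiving and jumping.
# An eased additive tilt offsets the pack during melee, emotes, skydiving or falling.
# A kill-stat watcher raises Active_Kill, briefly blending additive RotateBone and left-forearm rotations except while skydiving, before a delayed reset.
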AnimBlueprint /BRCosmetics/Accessories/FORT_Backpacks/Backpack_M_MED_EmeraldGlassGreen/Meshes/M_MED_EmeraldGlassGreen_Pack_AnimBP
target_skeleton=&/M_MED_EmeraldGlassGreen_Pack_Skeleton
graphs [
 AnimGraph {nodes [
   n1 = Get GravityOverride()
   n2 = RigidBody!(
    ComponentPose=Local To Component(LocalPose=Input Pose())
    Alpha=Get RigidBodyAlpha() -> (RigidBodyAlpha:float)
    OverrideWorldGravity:Vector=@n1
    ExternalForce
   ) -> (Pose) {
    bOverrideWorldGravity=true
    overridePhysicsAsset=&/M_MED_EmeraldGlassGreen_Pack_Physics
    alphaBoolBlend {blendOption=Linear}
    alphaScaleBiasClamp {bInterpResult=true, interpSpeedDecreasing=2, interpSpeedIncreasing=1}
    componentAppliedLinearAccClamp=<3000 3000 3000>
    componentLinearAccScale=<0.4 0.4 0.3>
    componentLinearVelScale=<0.4 0.4 0.3>
    overrideWorldGravity {z=-980}
   }
   n3 = RigidBody(ComponentPose=@n2, Alpha=Get LARM Alpha() -> (LARM Alpha:float), OverrideWorldGravity=@n1) {
    bOverrideWorldGravity=true
    overridePhysicsAsset=&/M_MED_EmeraldGlassGreen_Pack_Physics_LeftARM
    alphaBoolBlend {blendOption=Linear}
    alphaScaleBiasClamp {bInterpResult=true, interpSpeedDecreasing=2, interpSpeedIncreasing=1}
    componentAppliedLinearAccClamp=<3000 3000 3000>
    componentLinearAccScale=<0.4 0.4 0.3>
    componentLinearVelScale=<0.4 0.4 0.3>
    overrideWorldGravity {z=-980}
   }
   n4 = "Transform (Modify) Bone - Bone: pack"(ComponentPose=@n3, Alpha=Get Is Playing Melee Anim() -> (Is Playing Melee Anim:double), Translation="0,0,0", Rotation="0,-20,0") {
    rotationMode=BMM_Additive
    rotationSpace=BCS_BoneSpace
    translationMode=BMM_Additive
    translationSpace=BCS_BoneSpace
    alphaBoolBlend {blendOption=Linear}
    alphaScaleBiasClamp {bInterpResult=true, interpSpeedDecreasing=2, interpSpeedIncreasing=4}
    boneToModify {boneName=pack}
    rotation {yaw=-20}
   }
   n5 = Get Active_Kill() -> (Active_Kill:bool)
   n6 = MapRangeClamped(Value=SelectFloat(A=0, B="To Float (Boolean)"(InBool=@n5), bPickA=Get bIsSkyDiving()), InRangeA=0, InRangeB=1, OutRangeB=3)
   n7 = "Transform (Modify) Bone - Bone: RotateBone"(ComponentPose=@n4, Alpha=@n6, Translation="0,0,0", Rotation="-8.094,-8.8,140", Scale="1,1,1") {
    rotationMode=BMM_Additive
    rotationSpace=BCS_BoneSpace
    translationMode=BMM_Additive
    translationSpace=BCS_BoneSpace
    alphaBoolBlend {blendOption=Linear}
    alphaScaleBiasClamp {bInterpResult=true, interpSpeedDecreasing=1.7, interpSpeedIncreasing=4}
    boneToModify {boneName=RotateBone}
    rotation=<rot -8.094 -8.8 140>
   }
   n8 = "Transform (Modify) Bone - Bone: dyn_forarm_L"(ComponentPose=@n7, Alpha=@n6, Translation="0,0,0", Rotation="14.953,4.819,13.346", Scale="1,1,1") {
    rotationMode=BMM_Additive
    rotationSpace=BCS_BoneSpace
    translationMode=BMM_Additive
    translationSpace=BCS_BoneSpace
    alphaBoolBlend {blendOption=Linear}
    alphaScaleBiasClamp {bInterpResult=true, interpSpeedDecreasing=1.5, interpSpeedIncreasing=4}
    boneToModify {boneName=dyn_forarm_L}
    rotation=<rot 14.953 4.819 13.3459>
   }
   Output Pose(Result=Component To Local(ComponentPose=@n8)) {layerGroup=None}
   Get bIsJumping()
  ]}
 EventGraph {nodes [
   n1 = Event BlueprintUpdateAnimation()
   > n2 = Sequence!(execute) -> (then_0, then_1, then_2:Exec, then_3:Exec, then_4:Exec, then_5:Exec)
   > n3 = Gravity Override State Function DEPRECATED(CharacterPartAnimInstance=Self-Reference(), GameTimeInSeconds=GetGameTimeInSeconds(), GravityOverrideParams=Get Gravity Override Params())
   > Set GravityOverride(GravityOverride=@n3)
   n4 = "To Float (Boolean)"(InBool=OR Boolean!(A=Get bIsPlayingMeleeAnim(), B=Get bIsPlayingEmote(), C:bool=Get bIsSkyDiving(), D:bool=Get bIsSkydiveFloating(), E:bool=Get bIsFalling()) -> (ReturnValue))
   Set Is Playing Melee Anim(
    execute=@n2.then_1
    Is Playing Melee Anim:double=FInterpTo(
     Current=Get Is Playing Melee Anim()
     Target=MapRangeClamped(Value=@n4, InRangeA=0, InRangeB=1, OutRangeB=0.7)
     DeltaTime=@n1.DeltaTimeX
     InterpSpeed=0.45
    )
   ) -> (Output_Get:double)
   n5 = Branch(execute=@n2.then_3, Condition=Get bIsSkyDiving())
   > Set RigidBodyAlpha(RigidBodyAlpha:float=Get IsFalling Alpha() -> (IsFalling Alpha:float)) -> (Output_Get:float)
   Set RigidBodyAlpha(execute=@n5.else, RigidBodyAlpha=Get Regular Alpha() -> (Regular Alpha:float))
   n6 = Branch(execute=@n2.then_4, Condition=Get bIsJumping())
   > Set LARM Alpha(LARM Alpha:float=Get LARM Jumping() -> (LARM Jumping:float)) -> (Output_Get:float)
   Set LARM Alpha(execute=@n6.else, LARM Alpha=Get LARM Regular() -> (LARM Regular:float))
   Get!(execute:Exec=@n2.then_5) -> (then:Exec, else:Exec, Skel_Mesh:Skeletal Mesh Component Object Reference)
   > Sequence!(execute) -> (then_0, then_1)
   > n7 = Branch(Condition=Get Active_Kill())
   > "Set FInterp(Alpha, (10: 10))"("FInterp(Alpha, (10: 10))":float="Get FInterp(Alpha, (10: 10))_0"() -> ("FInterp(Alpha, (10: 10))_0":float)) -> (Output_Get:float)
   > Delay(Duration=1.5)
   > Set Active_Kill(Active_Kill:bool=false) -> (Output_Get:bool)
   "Set FInterp(Alpha, (10: 10))"(execute=@n7.else, "FInterp(Alpha, (10: 10))"=0)
   Event BlueprintInitializeAnimation()
   > Set Skel_Mesh(Skel_Mesh=GetOwnerSkeletalMeshComponent())
   Event OnCharacterCustomizationLoaded()
   > n8 = Cast To FortPlayerPawn(Object=TryGetFortPawn())
   > n9 = WatchCosmeticStat!(
    execute:Exec
    Pawn:Fort Player Pawn Object Reference=@n8
    StatTag:Fort Stat Manager Tag Structure=(TagName="StatManager.AthenaKills")
    DefaultValue:int=0
    bBroadcastInitialValue:bool=false
    InitialValueBroadcastDelay:float=-1
    ChangeValueBroadcastDelay:float=-1
    bConsumeWaitingNotifications:bool=true
   ) -> (then:Exec, OnStatChanged:Exec, StatName:Name, StatValue:int) {is=title K2Node_AsyncAction}
   Set KillCount(execute=@n9.OnStatChanged, KillCount:double="To Float (Integer)"(InInt=@n9.StatValue)) -> (Output_Get:double)
   > Set Active_Kill(Active_Kill=true)
   n10 = Branch(Condition=Get Active Kill() -> (Active Kill:bool))
   > "Set FInterp(Alpha, (10: 10))"("FInterp(Alpha, (10: 10))"="Get FInterp(Alpha, (10: 10))_0"())
   > PrintString!(execute, InString=INTERPOKill, bPrintToScreen, bPrintToLog, TextColor=(R=0.605,G=0.0772,B=1,A=1), Duration, Key) -> (then)
   > Delay(Duration=1.5)
   > Set Active_Kill(Active_Kill=false)
   "Set FInterp(Alpha, (10: 10))"(execute=@n10.else, "FInterp(Alpha, (10: 10))"=0)
  ]}
]
variables [
 Active Kill {type=boolean}
 Active_Kill {type=boolean}
 Did_Kill {type=boolean}
 "FInterp(Alpha, (10: 10))" {type=number, default=1}
 "FInterp(Alpha, (10: 10))_0" {type=number, default=1}
 Gravity Override Params {type=GravityOverrideParamsStruct}
 GravityOverride {type=Vector}
 Is Playing Melee Anim {type=number}
 IsFalling Alpha {type=number, default=0.5}
 KillCount {type=number}
 LARM Alpha {type=number, default=1}
 LARM Jumping {type=number, default=0.2}
 LARM Regular {type=number, default=1}
 Last_KillCount {type=number}
 Pawn {type=FortPlayerPawn}
 Regular Alpha {type=number, default=1}
 RigidBodyAlpha {type=number, default=1}
 Skel_Mesh {type=SkeletalMeshComponent}
 SkelMesh {type=SkeletalMesh, default=null}
]
